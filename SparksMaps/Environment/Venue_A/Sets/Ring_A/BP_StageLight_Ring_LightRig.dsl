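; Stage light rig Blueprint for the Venue A ring set that drives light-show effects on the SM_Ring_LightRig mesh.
; Activate, beat and impulse handlers, duplicated as events and MojoStageShow functions, write strength, pulse and time-signature values into custom primitive data for the material.
; Impulses stamp game time and snap the time signature to one of two values by threshold.
; Color shift, custom mode, stop and Niagara update functions are stubs returning false.
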(event EventStageShowActivateBeat
  (params (Active :type Float%20%28single-precision%29) (MusicTimeMeasure :type Float%20%28single-precision%29))
  (bind _sm_ring_lightrig (Variables|Default|GetSM_Ring_LightRig))
  (Rendering|Material|SetScalarParameterforCustomPrimitiveData _sm_ring_lightrig "BeatPulse_Active" Active)
  (Rendering|Material|SetScalarParameterforCustomPrimitiveData _sm_ring_lightrig "StageShow_MusicTimeSignature" MusicTimeMeasure))

(event EventStageShowActivate
  (params (Strength :type Float%20%28single-precision%29) (MusicTimeMeasure :type Float%20%28single-precision%29))
  (bind _sm_ring_lightrig (Variables|Default|GetSM_Ring_LightRig))
  (Rendering|Material|SetScalarParameterforCustomPrimitiveData _sm_ring_lightrig "StageShow_ActivateStrength" Strength)
  (Rendering|Material|SetScalarParameterforCustomPrimitiveData _sm_ring_lightrig "StageShow_MusicTimeSignature" MusicTimeMeasure))

(event EventStageShowImpulse
  (params (MusicTimeMeasure :type Float%20%28single-precision%29))
  (bind _sm_ring_lightrig (Variables|Default|GetSM_Ring_LightRig))
  (Rendering|Material|SetScalarParameterforCustomPrimitiveData _sm_ring_lightrig "StageShow_MusicTimeSignature" (Math|Float|SelectFloat 3.0 0.5 (> MusicTimeMeasure 0.6)))
  (Rendering|Material|SetScalarParameterforCustomPrimitiveData _sm_ring_lightrig "StageShow_ImpulseTime" (Utilities|Time|GetGameTimeinSeconds)))

(fn MojoStageShowActivate
  (params (Strength :type Float%20%28double-precision%29) (MusicTimeMeasure :type Float%20%28double-precision%29))
  (bind _sm_ring_lightrig (Variables|Default|GetSM_Ring_LightRig))
  (Rendering|Material|SetScalarParameterforCustomPrimitiveData _sm_ring_lightrig "StageShow_ActivateStrength" Strength)
  (Rendering|Material|SetScalarParameterforCustomPrimitiveData _sm_ring_lightrig "StageShow_MusicTimeSignature" MusicTimeMeasure)
  (return false))

(fn MojoStageShowActivateBeat
  (params (Active :type Float%20%28double-precision%29) (MusicTimeMeasure :type Float%20%28double-precision%29))
  (bind _sm_ring_lightrig (Variables|Default|GetSM_Ring_LightRig))
  (Rendering|Material|SetScalarParameterforCustomPrimitiveData _sm_ring_lightrig "BeatPulse_Active" Active)
  (Rendering|Material|SetScalarParameterforCustomPrimitiveData _sm_ring_lightrig "StageShow_MusicTimeSignature" MusicTimeMeasure)
  (return false))

(fn MojoStageShowColorShift
  (params (Primary :type Linear%20Color%20Structure) (Secondary :type Linear%20Color%20Structure) (Tertiary :type Linear%20Color%20Structure) (AltPrimary :type Linear%20Color%20Structure) (AltSecondary :type Linear%20Color%20Structure) (AltTertiary :type Linear%20Color%20Structure) (MusicTimeMeasure :type Float%20%28double-precision%29))
  (return false))

(fn MojoStageShowCustomMode
  (params (ModeIndex :type Integer) (MusicTimeMeasure :type Float%20%28double-precision%29))
  (return false))

(fn MojoStageShowImpulse
  (params (MusicTimeMeasure :type Float%20%28double-precision%29))
  (bind _sm_ring_lightrig (Variables|Default|GetSM_Ring_LightRig))
  (Rendering|Material|SetScalarParameterforCustomPrimitiveData _sm_ring_lightrig "StageShow_MusicTimeSignature" (Math|Float|SelectFloat 3.0 0.5 (> MusicTimeMeasure 0.6)))
  (Rendering|Material|SetScalarParameterforCustomPrimitiveData _sm_ring_lightrig "StageShow_ImpulseTime" (Utilities|Time|GetGameTimeinSeconds))
  (return false))

(fn MojoStageShowOnStop
  (return false))

(fn MojoStageShowUpdateNiagaraSystem
  (params (Niagara%20System :type Niagara%20System%20Object%20Reference))
  (return false))

(fn ConstructionScript)
